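# Translucent UI material that draws the mission icon for challenge tiles, with an optional prestige shine.
# The mask texture's red channel blends border and star colors over a shadow color, while its blue channel bounds the alpha that rises from the shadow's opacity to solid.
# A shared ShineSwipe function, masked by the red-derived shape, adds a tinted highlight that the PrestigeActive scalar fades in.
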Material /Game/Athena/UI/Challenges/Art/MissionTileImages/M_UI_Icon_Mission
graph {
 properties {
  bUsedWithStaticMesh=false
  blendMode=BLEND_Translucent
  materialDomain=MD_UI
  pixelDepthOffsetMode=PDOM_Legacy
  usedShadingModels=Default Lit
 }
 nodes [
  n1 = VectorParameter() {parameterName=ShadowRGBA, defaultValue=<rgba 0 0 0 0.5>}
  n2 = TextureSampleParameter2D(UVs=TextureCoordinate()) {parameterName=TextureMask, texture=&/T_UI_Icon_Mission}
  n3 = LinearInterpolate(A=@n2.R, Alpha=@n2.R)
  LinearInterpolate(B=LinearInterpolate(A=@n1.A, Alpha=@n3), Alpha=@n2.B)
  n4 = LinearInterpolate(
   A=@n1.RGB
   B=LinearInterpolate(
    A=VectorParameter() {parameterName=border, defaultValue=<rgba 1 0.2549 0 0>}.RGB
    B=VectorParameter() {parameterName=Star, defaultValue=<rgba 1 0.5863 0 0>}.RGB
    Alpha=@n2.R
   )
   Alpha=@n3
  )
  n5 = MaterialFunctionCall!(In_Uv) {materialFunction=/Game/UI/Foundation/Materials/M-Function-ShineSwipe}
  LinearInterpolate(
   A=@n4
   B=LinearInterpolate(
    A=@n4
    B=VectorParameter() {parameterName=ShineRGB, defaultValue=<rgba 1 0.9865 0.6908 0>}.RGB
    Alpha=LinearInterpolate(B=@n5.Result, Alpha=@n3)
   )
   Alpha=ScalarParameter() {parameterName=PrestigeActive}
  )
 ]
}
