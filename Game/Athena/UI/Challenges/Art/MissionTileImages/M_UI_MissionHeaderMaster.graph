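# Translucent UI master material for challenge mission header tiles, blending key art with color gradients inside a shaped mask.
# With UseKeyArt on, offset and scaled CustomBackground art fades into Secondary or Accent, then Primary, through horizontal gradients; otherwise only colors blend.
# IsGiftBox switches to unmodified UVs.
# The mask multiplies an optional procedural bolt edge, built from three rotated-UV cuts, by optional angled ends.
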Material /Game/Athena/UI/Challenges/Art/MissionTileImages/M_UI_MissionHeaderMaster
graph {
 properties {
  bUsedWithStaticMesh=false
  blendMode=BLEND_Translucent
  materialDomain=MD_UI
  pixelDepthOffsetMode=PDOM_Legacy
  usedShadingModels=Default Lit
 }
 nodes [
  n1 = Divide(A=ScalarParameter() {defaultValue=-4.5, group=BoltEdge, parameterName=LeftAngle}) {constB=360}
  n2 = TextureCoordinate() {coordinateIndex=4}
  n3 = ScalarParameter() {defaultValue=0.1, group=BoltEdge, parameterName=LeftOffset}
  n4 = ComponentMask(None=Add(A=Rotator(Coordinate=@n2, Time=@n1) {centerY=0, speed=6.2832}, B=Add(A=@n3) {constB=-1})) {g=true}
  n5 = ScalarParameter() {defaultValue=112, group=BoltEdge, parameterName=LeftRightEdgeSharpness}
  n6 = Divide(A=ScalarParameter() {defaultValue=-5.75, group=BoltEdge, parameterName=RightAngle}) {constB=360}
  n7 = ScalarParameter() {defaultValue=0.1875, group=BoltEdge, parameterName=RightOffset}
  n8 = ComponentMask(None=Add(A=Rotator(Coordinate=@n2, Time=@n6) {centerY=1, speed=6.2832}, B=Add(A=@n7) {constB=-1})) {g=true}
  n9 = Rotator(
   Coordinate=Add(A=@n2) {constB=-0.5}
   Time=Divide(A=ScalarParameter() {defaultValue=12, group=BoltEdge, parameterName=CenterAngle}) {constB=360}
  ) {speed=6.2832}
  n10 = Multiply(A=ScalarParameter() {defaultValue=0.0875, group=BoltEdge, parameterName=CenterOffset}) {constB=-1}
  n11 = Multiply(A=ComponentMask(None=Add(A=@n9, B=@n10)) {r=true}, B=ScalarParameter() {defaultValue=80, group=BoltEdge, parameterName=CenterEdgeSharpness})
  n12 = OneMinus(None=LinearInterpolate(A=Multiply(A=@n4, B=@n5), B=Multiply(A=@n8, B=@n5), Alpha=Clamp(None=@n11)))
  n13 = Divide(A=ScalarParameter() {defaultValue=-4.5, group=Ends, parameterName=LeftEndAngle}) {constB=360}
  n14 = TextureCoordinate()
  n15 = ScalarParameter() {defaultValue=0.1, group=Ends, parameterName=LeftEndOffset}
  n16 = ComponentMask(None=Add(A=Rotator(Coordinate=@n14, Time=@n13) {centerX=1, speed=6.2832}, B=Multiply(A=@n15) {constB=-1})) {r=true}
  n17 = ScalarParameter() {defaultValue=8, group=Ends, parameterName=RightLeftEndSharpness}
  n18 = Divide(A=ScalarParameter() {defaultValue=-5.75, group=Ends, parameterName=RightEndAngle}) {constB=360}
  n19 = ScalarParameter() {defaultValue=0.1875, group=Ends, parameterName=RightEndOffset}
  n20 = ComponentMask(None=Add(A=Rotator(Coordinate=@n14, Time=@n18) {centerX=0, speed=6.2832}, B=Add(A=@n19) {constB=-1})) {r=true}
  n21 = OneMinus(None=DotProduct(A=Multiply(A=@n16, B=@n17), B=Multiply(A=@n20, B=@n17)))
  LinearInterpolate(
   B=StaticSwitchParameter(True=Clamp(None=@n12), False=Constant() {r=1}) {defaultValue=true, parameterName=UseBoltMask}
   Alpha=LinearInterpolate(A=Constant() {r=1}, B=Clamp(None=@n21), Alpha=ScalarParameter() {group=Ends, parameterName=UseAngledEnds})
  )
  n22 = TextureCoordinate()
  n23 = Constant() {r=-0.2632}
  n24 = Add(A=@n22, B=AppendVector(A=@n23, B=Constant()))
  n25 = MaterialFunctionCall(UVs=@n24, Texture Scale=AppendVector(A=Constant() {r=1}, B=Constant() {r=1.2727})) {materialFunction=&ScaleUVsByCenter}
  n26 = Constant() {r=1.3571}
  n27 = AppendVector(A=@n26, B=Constant() {r=1})
  n28 = TextureSampleParameter2D(UVs=LinearInterpolate(A=Multiply(A=@n25.UVs, B=@n27), B=TextureCoordinate(), Alpha=ScalarParameter() {parameterName=IsGiftBox})) {parameterName=CustomBackground, texture=/Game/Athena/UI/Challenges/Art/TileImages/T_UI_ChallengeTile_Discovery_CustomBackground}
  n29 = LinearInterpolate(
   A=VectorParameter() {parameterName=Secondary, defaultValue=<rgba 0.0356 0.0065 0.0052 1>}.RGB
   B=VectorParameter() {parameterName=Accent, defaultValue=<rgba 1 0.1779 0.0529 1>}.RGB
   Alpha=ScalarParameter() {parameterName=UseAccentoverSecondary}
  )
  n30 = Constant() {r=1}
  n31 = LinearInterpolate(A=Multiply(A=@n24, B=AppendVector(A=@n26, B=@n30)), B=TextureCoordinate(), Alpha=ScalarParameter() {parameterName=IsGiftBox})
  n32 = Clamp(
   None=OneMinus(None=ComponentMask(None=@n31) {r=true})
   Min=ScalarParameter() {parameterName=SecondaryGradientStart}
   Max=ScalarParameter() {defaultValue=1, parameterName=SecondaryGradientStop}
  )
  n33 = VectorParameter() {parameterName=Primary, defaultValue=<rgba 0.3515 0.0452 0.0212 1>}
  n34 = AppendVector(A=Multiply(A=@n23) {constB=-1}, B=Constant())
  n35 = Constant() {r=1}
  n36 = LinearInterpolate(A=@n22, B=TextureCoordinate(), Alpha=ScalarParameter() {parameterName=IsGiftBox})
  n37 = ScalarParameter() {parameterName=PrimaryGradientOffset}
  n38 = StaticSwitchParameter(
   True=Multiply(A=Add(A=@n24, B=@n34), B=AppendVector(A=@n26, B=@n35))
   False=Multiply(A=Add(A=@n36, B=@n37), B=ScalarParameter() {defaultValue=1, parameterName=PrimaryGradientScale})
  ) {defaultValue=true, parameterName=UseKeyArt}
  n39 = Clamp(
   None=OneMinus(None=ComponentMask(None=@n38) {r=true})
   Min=ScalarParameter() {parameterName=PrimaryGradientStart}
   Max=ScalarParameter() {defaultValue=1, parameterName=PrimaryGradientStop}
  )
  StaticSwitchParameter(
   True=LinearInterpolate(A=LinearInterpolate(A=@n28.RGB, B=@n29, Alpha=@n32), B=@n33.RGB, Alpha=@n39)
   False=LinearInterpolate(A=@n29, B=@n33.RGB, Alpha=@n39)
  ) {defaultValue=true, parameterName=UseKeyArt}
 ]
}
